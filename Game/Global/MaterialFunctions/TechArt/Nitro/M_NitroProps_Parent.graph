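# Parent material for Nitro props, a Default Lit surface built on material attributes and flagged for Nanite use.
# A SetMaterialAttributes node feeds the shared MF_Nitro_v2 function, whose result passes through MF_CreativeMoveTool.
# The graph is a thin wrapper with its logic in those two functions, and automatic usage flagging in the editor is off.
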Material /Game/Global/MaterialFunctions/TechArt/Nitro/M_NitroProps_Parent
graph {
 properties {
  bAutomaticallySetUsageInEditor=false
  bUseMaterialAttributes=true
  bUsedWithNanite=true
  pixelDepthOffsetMode=PDOM_Legacy
  usedShadingModels=Default Lit
 }
 nodes [MaterialFunctionCall(MA In=MaterialFunctionCall!(MA=SetMaterialAttributes!(MaterialAttributes), SkinFX) {materialFunction=/Game/Global/Materials/MaterialGym/Materials/Nitro/MF_Nitro_v2}.Result) {materialFunction=&MF_CreativeMoveTool}]
}
