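# Prototype material for a glowing Nitro liquid, two-sided and usable on skeletal and Niagara meshes.
# A custom HLSL loop ray-marches scrolling, distorted noise for height and normals, with a variant blended across three pre-skinned projection planes.
# Parallax bubble layers, Fresnel and a time-hashed Volatile flicker shape emissive color, alongside cylinder-intersection experiments for a top mask.
# Many branches are loose experiments beside an MF_NitroLiquid call.
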Material /Game/Global/MaterialFunctions/TechArt/Nitro/M_NitroProto
graph {
 properties {
  bHasPixelAnimation=true
  bUseMaterialAttributes=true
  bUsedWithMeshDeformer=true
  bUsedWithNiagaraMeshParticles=true
  bUsedWithSkeletalMesh=true
  pixelDepthOffsetMode=PDOM_Legacy
  shadingModel=MSM_FromMaterialExpression
  translucencyLightingMode=TLM_SurfacePerPixelLighting
  twoSided=true
  usedShadingModels=Default Lit
 }
 nodes [
  n1 = MaterialFunctionCall(Float3=VectorParameter() {parameterName=XYScvale, defaultValue=<rgba -32 0 0 1>}.RGB) {materialFunction=&BreakOutFloat3Components}
  n2 = NamedRerouteUsage()
  n3 = Saturate(None=Power(Base=OneMinus(None=@n2)) {constExponent=12})
  Add(A=@n1.G, B=Multiply(A=Max(A=@n2, B=@n3)) {constB=24})
  n4 = Constant3Vector()
  n5 = MaterialFunctionCall!(
   Origin=TransformPosition(None=@n4) {transformType=TRANSFORMPOSSOURCE_World}
   Radius=ScalarParameter() {defaultValue=128, parameterName=Radius}
   Z Axis=Transform(None=Constant3Vector() {constant=<rgba 0 0 1 0>}) {transformSourceType=TRANSFORMSOURCE_Local}
   XY Scale
   X Axis
   Capped?=StaticBool() {value=true}
   Z Min-Max=MaterialFunctionCall(X=@n1.R, Y=@n1.R) {materialFunction=&MakeFloat2}.Result
  ) {materialFunction=/Engine/Functions/Engine_MaterialFunctions01/Math/CylinderIntersection}
  ComponentMask(None=@n5.Opacity Mask) {b=true, g=true, r=true}
  n6 = Divide(A=VertexInterpolator(VS=PreSkinnedPosition()).PS, B=ScalarParameter() {defaultValue=24, parameterName=BaseScale})
  ComponentMask(None=@n6) {g=true, r=true}
  n7 = Multiply(
   A=BumpOffset(Coordinate=NamedRerouteUsage(), Height=Constant() {r=-1}) {heightRatio=0.1}
   B=ComponentMask(None=VectorParameter() {parameterName=BubbleTiling, defaultValue=<rgba 0.5 1 0 1>}.RGB) {g=true, r=true}
  ) {constB=3}
  n8 = NamedRerouteUsage()
  n9 = Multiply(A=Time(), B=ComponentMask(None=VectorParameter() {parameterName=NS_B_PanSpeed, defaultValue=<rgba 0 0.25 0 1>}.RGB) {g=true, r=true})
  n10 = ComponentMask(None=VectorParameter() {parameterName=NS_B_Tiling, defaultValue=<rgba 10 10 10 1>}.RGB) {g=true, r=true}
  n11 = TextureSample(UVs=Multiply(A=Add(A=@n8, B=@n9), B=Multiply(A=@n10))) {texture=/Game/Characters/Player/Male/Medium/Bodies/M_MED_Viper/Textures/Glitch/T_SoftNoise}
  n12 = ComponentMask(None=Subtract(A=Multiply(A=@n11.RGB) {constB=2})) {g=true, r=true}
  n13 = ScalarParameter() {defaultValue=0.01, parameterName=NS_B_DistortionStrength}
  n14 = TextureSample(UVs=Add(A=Panner(Coordinate=@n7) {speedY=-1}, B=Multiply(A=@n12, B=@n13))) {samplerType=SAMPLERTYPE_LinearGrayscale, texture=&/T_NitroBubles}
  ComponentMask(None=@n14.RGB) {g=true}
  Constant3Vector() {constant=<rgba 0.5 0.5 1 0>}
  n15 = ScalarParameter() {defaultValue=1.5, parameterName=NitroPow}
  n16 = ScalarParameter() {defaultValue=1, parameterName=NitroVolatilePow}
  n17 = ComponentMask(None=NamedRerouteUsage()) {r=true}
  n18 = Time()
  n19 = ScalarParameter() {defaultValue=1, parameterName=L_01_Speed}
  n20 = MaterialFunctionCall(Float=Frac(None=Multiply(A=@n18, B=@n19))) {materialFunction=&MF_Hash}
  n21 = Multiply(
   A=LinearInterpolate(A=Power(Base=@n17), B=@n17, Alpha=NamedRerouteUsage())
   B=LinearInterpolate(A=Saturate(None=@n20.Float), Alpha=NamedRerouteUsage())
  )
  n22 = Multiply(A=@n21, B=LinearInterpolate(A=NamedRerouteUsage(), Alpha=NamedRerouteUsage()))
  n23 = Saturate(None=Power(Base=NamedRerouteUsage(), Exp=LinearInterpolate(A=@n15, B=@n16, Alpha=@n22) {constB=0}) {constExponent=4})
  CurveAtlasRowParameter(CurveTime=@n23) {
   atlas=&/CA_NitroColors
   curve=&/NitroLiquidColor
   parameterName=NitroLiquidColor
  }
  Frac(None=Multiply(A=Time()) {constB=0.25})
  Frac(None=Multiply(A=Time()) {constB=0.25})
  n24 = BumpOffset(Coordinate=NamedRerouteUsage(), Height=Constant() {r=-1}, HeightRatioInput=ScalarParameter() {defaultValue=0.3, parameterName=HR_2}) {heightRatio=0.1}
  n25 = ComponentMask(None=VectorParameter() {parameterName=BubbleTiling_2, defaultValue=<rgba 1 2 0 1>}.RGB) {g=true, r=true}
  n26 = VectorParameter() {parameterName=NS_B_PanSpeed, defaultValue=<rgba 0 0.25 0 1>}
  n27 = Add(A=NamedRerouteUsage(), B=Multiply(A=Time(), B=ComponentMask(None=@n26.RGB) {g=true, r=true}))
  n28 = Multiply(A=ComponentMask(None=VectorParameter() {parameterName=NS_B_Tiling, defaultValue=<rgba 10 10 10 1>}.RGB) {g=true, r=true})
  n29 = Multiply(A=TextureSample(UVs=Multiply(A=@n27, B=@n28)) {texture=&T_SoftNoise}.RGB) {constB=2}
  n30 = Multiply(A=ComponentMask(None=Subtract(A=@n29)) {g=true, r=true}, B=ScalarParameter() {defaultValue=0.01, parameterName=NS_B_DistortionStrength})
  n31 = Add(A=Panner(Coordinate=Multiply(A=@n24, B=@n25) {constB=3}) {speedY=-0.4}, B=@n30)
  n32 = BumpOffset(Coordinate=NamedRerouteUsage(), Height=Constant() {r=-1}, HeightRatioInput=ScalarParameter() {defaultValue=0.2, parameterName=HR_1}) {heightRatio=0.1}
  n33 = ComponentMask(None=VectorParameter() {parameterName=BubbleTiling, defaultValue=<rgba 0.5 1 0 1>}.RGB) {g=true, r=true}
  n34 = Add(A=Panner(Coordinate=Multiply(A=@n32, B=@n33) {constB=3}) {speedY=-0.4}, B=@n30)
  n35 = TextureSample(UVs=@n34) {samplerType=SAMPLERTYPE_LinearGrayscale, texture=&T_NitroBubles}
  n36 = LinearInterpolate(
   A=TextureSample(UVs=@n31) {samplerType=SAMPLERTYPE_LinearGrayscale, texture=&T_NitroBubles}.G
   B=@n35.G
   Alpha=@n35.G
  )
  n37 = Constant() {r=1}
  n38 = NamedRerouteUsage()
  n39 = TextureSample(UVs=Panner(Coordinate=Multiply(A=@n38) {constB=2}) {speedY=0.1}) {texture=&T_SoftNoise}
  n40 = LinearInterpolate(Alpha=Saturate(None=Power(Base=@n39.R) {constExponent=1})) {constA=0.5}
  n41 = Constant()
  n42 = Saturate(None=Multiply(A=SphereMask(A=@n36, B=@n37, Radius=@n40, Hardness=@n41)) {constB=5})
  n43 = TextureSample(UVs=@n31) {samplerType=SAMPLERTYPE_Normal, texture=&/T_NitroBubles_N}
  n44 = TextureSample(UVs=@n34) {samplerType=SAMPLERTYPE_Normal, texture=&T_NitroBubles_N}
  MaterialExpressionMultiply_33 = Multiply(A=NamedRerouteUsage())
  MaterialExpressionTransform_1 = Transform(None=CameraVectorWS()) {transformSourceType=TRANSFORMSOURCE_World, transformType=TRANSFORM_Tangent}
  MaterialExpressionConstant_13 = Constant() {r=16}
  MaterialExpressionConstant_12 = Constant() {r=0.05}
  MaterialExpressionTextureObject_1 = TextureObject() {
   samplerType=SAMPLERTYPE_LinearGrayscale
   texture=/Game/Global/Materials/MaterialGym/Materials/SmokeWithFlowmap/Textures/T_Swirly_Noise_MiddleDetail
  }
  MaterialExpressionVertexNormalWS_1 = VertexNormalWS()
  MaterialExpressionTextureObject_0 = TextureObject() {
   samplerType=SAMPLERTYPE_LinearGrayscale
   texture=&T_Swirly_Noise_MiddleDetail
  }
  MaterialExpressionTime_4 = Time()
  n45 = Custom!(
   uv=@MaterialExpressionMultiply_33
   viewDir=@MaterialExpressionTransform_1
   steps=@MaterialExpressionConstant_13
   stepdist=@MaterialExpressionConstant_12
   tex=@MaterialExpressionTextureObject_1
   vertnormal=@MaterialExpressionVertexNormalWS_1
   dTex=@MaterialExpressionTextureObject_0
   Time=@MaterialExpressionTime_4
  ) {
   code="float3 rayStep = viewDir * -1;\r\nfloat2 uv2 = uv + Time*float2(0.0,-0.2);\r\nuv += Time*float2(0.0,-0.1);\r\nfloat4 heightTex = Texture2DSample(tex,texSampler,uv);\r\nfloat4 distortionTex = Texture2DSample(dTex,texSampler,uv2);\r\n//float steps = 16;\r\n//float stepdist = float(0.1);\r\n\r\nfor(int i=0;i<=16;i++)\r\n{\r\nif(heightTex.r<=i/16.0)\r\n//if(heightTex.r<=0.1)\r\n{\r\nreturn float((heightTex.r*i)/16.0)*1.0; \r\n//return float(i/16.0);\r\n}\r\nfloat stepct = heightTex.r * 0.025;\r\n//float stepct = 0.05;\r\nuv += rayStep * stepct;\r\nuv2 += rayStep * stepct;\r\ndistortionTex = Texture2DSample(dTex,texSampler,uv2*1.0);\r\nfloat2 distortStr = distortionTex.xy *0.1;\r\nheightTex = Texture2DSample(tex,texSampler,uv.xy + distortStr);\r\nRayMarchUV = uv.xy + distortStr;\r\n}\r\n\r\nreturn heightTex;\r\n"
   description=RayMarchTexture2
   outputType=CMOT_Float1
   showCode=true
   additionalOutputs [{outputName=RayMarchUV, outputType=CMOT_Float2}]
   inputs [
    {inputName=uv, input {expression=&self:MaterialExpressionMultiply_33}}
    {inputName=viewDir, input {expression=&self:MaterialExpressionTransform_1}}
    {inputName=steps, input {expression=&self:MaterialExpressionConstant_13}}
    {inputName=stepdist, input {expression=&self:MaterialExpressionConstant_12}}
    {inputName=tex, input {expression=&self:MaterialExpressionTextureObject_1}}
    {inputName=vertnormal, input {expression=&self:MaterialExpressionVertexNormalWS_1}}
    {inputName=dTex, input {expression=&self:MaterialExpressionTextureObject_0}}
    {inputName=Time, input {expression=&self:MaterialExpressionTime_4}}
   ]
  }
  n46 = TextureSample(UVs=@n45.RayMarchUV) {samplerType=SAMPLERTYPE_Normal, texture=/Game/Characters/Player/Male/Large/Bodies/M_LRG_Clash_V/Textures/Simple_Perlin_N}
  n47 = NamedRerouteUsage()
  LinearInterpolate(
   A=Transform(None=VertexNormalWS()) {transformSourceType=TRANSFORMSOURCE_World, transformType=TRANSFORM_Tangent}
   B=LinearInterpolate(A=@n46.RGB, B=LinearInterpolate(A=@n43.RGB, B=@n44.RGB, Alpha=@n35.G), Alpha=@n42)
   Alpha=@n47
  ) {constAlpha=1}
  n48 = Fresnel() {exponent=0.5}
  n49 = Saturate(None=Power(Base=OneMinus(None=@n48)))
  n50 = OneMinus(None=@n49)
  LinearInterpolate(A=LinearInterpolate(Alpha=Saturate(None=@n50)) {constA=0.45}, Alpha=NamedRerouteUsage())
  n51 = LinearInterpolate(
   A=VectorParameter() {parameterName=G_Color_02, defaultValue=<rgba 0.0156 0.0054 0.0016 1>}.RGB
   B=VectorParameter() {parameterName=G_Color_01, defaultValue=<rgba 0.4583 0.0978 0.0046 1>}.RGB
   Alpha=@n49
  )
  n52 = LinearInterpolate(Alpha=@n49) {constA=0.5}
  n53 = VectorParameter() {parameterName=EmisiveColor, defaultValue=<rgba 1 0.3866 0 1>}
  n54 = Multiply(A=LinearInterpolate(A=VectorParameter() {parameterName=EmisiveColor2, defaultValue=<rgba 0.0208 0.0019 0.0006 1>}.RGB, B=@n53.RGB, Alpha=@n23)) {constB=3}
  n55 = Multiply(
   A=LinearInterpolate(A=VectorParameter() {parameterName=EmisiveColor4, defaultValue=<rgba 0.3177 0.0283 0.0093 1>}.RGB, B=@n53.RGB, Alpha=@n23)
   B=Constant() {r=10}
  )
  n56 = VectorParameter() {parameterName=BubbleCOlor, defaultValue=<rgba 0.0208 0.0019 0.0006 1>}
  n57 = LinearInterpolate(
   A=LinearInterpolate(A=Power(Base=@n54) {constExponent=1.5}, B=Power(Base=@n55) {constExponent=3}, Alpha=NamedRerouteUsage()) {constB=0.5}
   B=Multiply(A=Multiply(A=@n42, B=@n56.RGB))
   Alpha=@n42
  )
  LinearInterpolate(
   A=LinearInterpolate(A=@n51, B=Multiply(A=@n52, B=@n57), Alpha=@n47) {constAlpha=1}
   B=Multiply(A=@n51, B=ScalarParameter() {defaultValue=0.5, parameterName=TopColorBlend})
   Alpha=NamedRerouteUsage()
  )
  MaterialExpressionMultiply_47 = Multiply(A=ComponentMask(None=@n6) {b=true, g=true})
  MaterialExpressionTransform_3 = Transform(None=CameraVectorWS()) {transformSourceType=TRANSFORMSOURCE_World, transformType=TRANSFORM_Tangent}
  MaterialExpressionConstant_8 = Constant() {r=16}
  MaterialExpressionConstant_5 = Constant() {r=0.05}
  MaterialExpressionTextureObject_2 = TextureObject() {
   samplerType=SAMPLERTYPE_LinearGrayscale
   texture=&T_Swirly_Noise_MiddleDetail
  }
  MaterialExpressionVertexNormalWS_2 = VertexNormalWS()
  MaterialExpressionTextureObject_3 = TextureObject() {
   samplerType=SAMPLERTYPE_LinearGrayscale
   texture=&T_Swirly_Noise_MiddleDetail
  }
  MaterialExpressionTime_8 = Time()
  n58 = Custom(
   uv=@MaterialExpressionMultiply_47
   viewDir=@MaterialExpressionTransform_3
   steps=@MaterialExpressionConstant_8
   stepdist=@MaterialExpressionConstant_5
   tex=@MaterialExpressionTextureObject_2
   vertnormal=@MaterialExpressionVertexNormalWS_2
   dTex=@MaterialExpressionTextureObject_3
   Time=@MaterialExpressionTime_8
  ) {
   code="float3 rayStep = viewDir * -1;\r\nfloat2 uv2 = uv + Time*float2(0.0,-0.2);\r\nuv += Time*float2(0.0,-0.1);\r\nfloat4 heightTex = Texture2DSample(tex,texSampler,uv);\r\nfloat4 distortionTex = Texture2DSample(dTex,texSampler,uv2);\r\n//float steps = 16;\r\n//float stepdist = float(0.1);\r\n\r\nfor(int i=0;i<=16;i++)\r\n{\r\nif(heightTex.r<=i/16.0)\r\n//if(heightTex.r<=0.1)\r\n{\r\nreturn float((heightTex.r*i)/16.0)*1.0; \r\n//return float(i/16.0);\r\n}\r\nfloat stepct = heightTex.r * 0.025;\r\n//float stepct = 0.05;\r\nuv += rayStep * stepct;\r\nuv2 += rayStep * stepct;\r\ndistortionTex = Texture2DSample(dTex,texSampler,uv2*1.0);\r\nfloat2 distortStr = distortionTex.xy *0.1;\r\nheightTex = Texture2DSample(tex,texSampler,uv.xy + distortStr);\r\nRayMarchUV = uv.xy + distortStr;\r\n}\r\n\r\nreturn heightTex;\r\n"
   description=RayMarchTexture2
   outputType=CMOT_Float1
   showCode=true
   additionalOutputs [{outputName=RayMarchUV, outputType=CMOT_Float2}]
   inputs [
    {inputName=uv, input {expression=&self:MaterialExpressionMultiply_47}}
    {inputName=viewDir, input {expression=&self:MaterialExpressionTransform_3}}
    {inputName=steps, input {expression=&self:MaterialExpressionConstant_8}}
    {inputName=stepdist, input {expression=&self:MaterialExpressionConstant_5}}
    {inputName=tex, input {expression=&self:MaterialExpressionTextureObject_2}}
    {inputName=vertnormal, input {expression=&self:MaterialExpressionVertexNormalWS_2}}
    {inputName=dTex, input {expression=&self:MaterialExpressionTextureObject_3}}
    {inputName=Time, input {expression=&self:MaterialExpressionTime_8}}
   ]
  }
  n59 = Abs(None=VertexInterpolator(VS=PreSkinnedNormal()).PS)
  n60 = ComponentMask(None=@n59) {b=true}
  MaterialExpressionMultiply_49 = Multiply(A=ComponentMask(None=@n6) {b=true, r=true})
  MaterialExpressionTransform_4 = Transform(None=CameraVectorWS()) {transformSourceType=TRANSFORMSOURCE_World, transformType=TRANSFORM_Tangent}
  MaterialExpressionConstant_14 = Constant() {r=16}
  MaterialExpressionConstant_9 = Constant() {r=0.05}
  MaterialExpressionTextureObject_4 = TextureObject() {
   samplerType=SAMPLERTYPE_LinearGrayscale
   texture=&T_Swirly_Noise_MiddleDetail
  }
  MaterialExpressionVertexNormalWS_3 = VertexNormalWS()
  MaterialExpressionTextureObject_5 = TextureObject() {
   samplerType=SAMPLERTYPE_LinearGrayscale
   texture=&T_Swirly_Noise_MiddleDetail
  }
  MaterialExpressionTime_9 = Time()
  n61 = Custom(
   uv=@MaterialExpressionMultiply_49
   viewDir=@MaterialExpressionTransform_4
   steps=@MaterialExpressionConstant_14
   stepdist=@MaterialExpressionConstant_9
   tex=@MaterialExpressionTextureObject_4
   vertnormal=@MaterialExpressionVertexNormalWS_3
   dTex=@MaterialExpressionTextureObject_5
   Time=@MaterialExpressionTime_9
  ) {
   code="float3 rayStep = viewDir * -1;\r\nfloat2 uv2 = uv + Time*float2(0.0,-0.2);\r\nuv += Time*float2(0.0,-0.1);\r\nfloat4 heightTex = Texture2DSample(tex,texSampler,uv);\r\nfloat4 distortionTex = Texture2DSample(dTex,texSampler,uv2);\r\n//float steps = 16;\r\n//float stepdist = float(0.1);\r\n\r\nfor(int i=0;i<=16;i++)\r\n{\r\nif(heightTex.r<=i/16.0)\r\n//if(heightTex.r<=0.1)\r\n{\r\nreturn float((heightTex.r*i)/16.0)*1.0; \r\n//return float(i/16.0);\r\n}\r\nfloat stepct = heightTex.r * 0.025;\r\n//float stepct = 0.05;\r\nuv += rayStep * stepct;\r\nuv2 += rayStep * stepct;\r\ndistortionTex = Texture2DSample(dTex,texSampler,uv2*1.0);\r\nfloat2 distortStr = distortionTex.xy *0.1;\r\nheightTex = Texture2DSample(tex,texSampler,uv.xy + distortStr);\r\nRayMarchUV = uv.xy + distortStr;\r\n}\r\n\r\nreturn heightTex;\r\n"
   description=RayMarchTexture2
   outputType=CMOT_Float1
   showCode=true
   additionalOutputs [{outputName=RayMarchUV, outputType=CMOT_Float2}]
   inputs [
    {inputName=uv, input {expression=&self:MaterialExpressionMultiply_49}}
    {inputName=viewDir, input {expression=&self:MaterialExpressionTransform_4}}
    {inputName=steps, input {expression=&self:MaterialExpressionConstant_14}}
    {inputName=stepdist, input {expression=&self:MaterialExpressionConstant_9}}
    {inputName=tex, input {expression=&self:MaterialExpressionTextureObject_4}}
    {inputName=vertnormal, input {expression=&self:MaterialExpressionVertexNormalWS_3}}
    {inputName=dTex, input {expression=&self:MaterialExpressionTextureObject_5}}
    {inputName=Time, input {expression=&self:MaterialExpressionTime_9}}
   ]
  }
  n62 = ComponentMask(None=@n59) {g=true}
  LinearInterpolate(A=LinearInterpolate(A=@n58.return, B=@n45.return, Alpha=@n60), B=@n61.return, Alpha=@n62)
  LinearInterpolate(
   A=LinearInterpolate(
    A=TextureSample(UVs=@n58.RayMarchUV) {samplerType=SAMPLERTYPE_Normal, texture=&Simple_Perlin_N}.RGB
    B=@n46.RGB
    Alpha=@n60
   )
   B=TextureSample(UVs=@n61.RayMarchUV) {samplerType=SAMPLERTYPE_Normal, texture=&Simple_Perlin_N}.RGB
   Alpha=@n62
  )
  MaterialFunctionCall!(Position, UVRadialScale) {materialFunction=/Game/Weapons/WeaponSkins/MaterialFunction/ObjectStabilizedScreenSpaceUVs}
  MaterialFunctionCall() {materialFunction=&CameraDirectionVector}
  MaterialFunctionCall() {materialFunction=&CameraDirectionVector}
  MaterialFunctionCall() {materialFunction=&ObjectStabilizedScreenSpaceUVs}
  MaterialFunctionCall() {materialFunction=&ObjectPivotPoint}
  MaterialFunctionCall(Position=TransformPosition(None=Constant3Vector()) {transformType=TRANSFORMPOSSOURCE_World}) {materialFunction=&ObjectStabilizedScreenSpaceUVs}
  MaterialFunctionCall() {materialFunction=&CameraDirectionVector}
  MaterialFunctionCall!(Material, Mask) {materialFunction=&/MF_NitroLiquid}
  n63 = Constant() {r=1}
  n64 = OneMinus(None=ComponentMask(None=NamedRerouteUsage()) {g=true})
  n65 = Multiply(A=TextureSample(UVs=Panner() {speedY=0.1}) {texture=&T_SoftNoise}.R) {constB=0.5}
  n66 = LinearInterpolate(Alpha=Saturate(None=Subtract(A=@n64, B=@n65))) {constA=0.75, constB=0.2}
  n67 = Constant()
  Multiply(
   A=Multiply(
    A=SphereMask(A=@n14.G, B=@n63, Radius=@n66, Hardness=@n67)
    B=VectorParameter() {parameterName=DefaultColor, defaultValue=<rgba 1 0.3281 0 1>}.RGB
   )
   B=ScalarParameter() {defaultValue=1, parameterName=BubbleBrightness}
  )
  n68 = VectorParameter() {parameterName=NS_2_PanSpeed, defaultValue=<rgba 0 -0.1 0 1>}
  n69 = Add(A=NamedRerouteUsage(), B=Multiply(A=Time(), B=ComponentMask(None=@n68.RGB) {g=true, r=true}))
  n70 = Multiply(A=ComponentMask(None=VectorParameter() {parameterName=NS_2_Tiling, defaultValue=<rgba 1 1 0 1>}.RGB) {g=true, r=true}, B=NamedRerouteUsage())
  n71 = NamedRerouteUsage()
  n72 = Multiply(A=Time(), B=ComponentMask(None=VectorParameter() {parameterName=NS_1_PanSpeed, defaultValue=<rgba 0 -0.25 0 1>}.RGB) {g=true, r=true})
  n73 = ComponentMask(None=VectorParameter() {parameterName=NS_1_Tiling, defaultValue=<rgba 1 1 0 1>}.RGB) {g=true, r=true}
  n74 = Multiply(A=NamedRerouteUsage()) {constB=2}
  n75 = TextureSample(UVs=Multiply(A=Add(A=@n71, B=@n72), B=Multiply(A=@n73, B=@n74))) {texture=&T_SoftNoise}
  n76 = ComponentMask(None=Subtract(A=Multiply(A=@n75.RGB) {constB=2})) {g=true, r=true}
  n77 = ScalarParameter() {defaultValue=0.1, parameterName=NS_1_DistortionStrength}
  n78 = TextureSample(UVs=Add(A=Multiply(A=@n69, B=@n70), B=Multiply(A=@n76, B=@n77))) {texture=&T_SoftNoise}
  Multiply(
   A=Multiply(A=StaticComponentMaskParameter(None=@n78.RGB) {defaultR=true}, B=ScalarParameter() {defaultValue=1, parameterName=NS2_Strength})
   B=Constant3Vector() {constant=<rgba 1 0.4849 0 1>}
  )
  Multiply(A=@n53.RGB) {constB=3}
  NamedRerouteDeclaration(None=AppendVector(A=ScalarParameter() {defaultValue=3, parameterName=SwirlTiling_X}, B=ScalarParameter() {defaultValue=1, parameterName=SwirlTiling_Y})) {name=NS-Tiling, nodeColor=<rgba 0.0078 1 0 1>}
  n79 = Power(Base=Multiply(A=NamedRerouteUsage()) {constB=2})
  n80 = NamedRerouteUsage()
  NamedRerouteDeclaration(None=Multiply(A=@n21, B=LinearInterpolate(A=@n79, Alpha=@n80))) {name=Lightning/VolatileMask, nodeColor=<rgba 1 0.1172 0 1>}
  NamedRerouteDeclaration(None=Constant() {r=1}) {name=CylinderMask, nodeColor=<rgba 0 1 0.0156 1>}
  NamedRerouteDeclaration(None=@n45.return) {name=RayMarchedHeightMap, nodeColor=<rgba 0 0.6719 1 1>}
  NamedRerouteDeclaration(None=ScalarParameter() {parameterName=Volatile}) {name=Volatile, nodeColor=<rgba 1 0 0.1641 1>}
  n81 = Constant3Vector() {constant=<rgba 0 0 1 0>}
  n82 = Multiply(A=@n5.Opacity Mask, B=@n5.Opacity Mask)
  NamedRerouteDeclaration(None=Saturate(None=DotProduct(A=@n81, B=@n82))) {name=TOpMask, nodeColor=<rgba 0.8047 1 0 1>}
  NamedRerouteDeclaration(None=TextureCoordinate()) {name=UV, nodeColor=<rgba 0 1 0.0391 1>}
  ObjectPositionWS()
  n83 = Constant3Vector()
  n84 = VectorParameter() {parameterName=CylinderPivotOffset, defaultValue=<rgba 0 0 64 1>}
  n85 = Subtract(A=WorldPosition().XYZ, B=TransformPosition(None=Subtract(A=@n83, B=@n84.RGB)) {transformType=TRANSFORMPOSSOURCE_World})
  n86 = Length(None=ComponentMask(None=@n85) {g=true, r=true}) {is=class MaterialExpressionLength}
  n87 = ComponentMask(None=@n85) {b=true}
  n88 = Subtract(
   A=Abs(None=AppendVector(A=@n86, B=@n87))
   B=AppendVector(A=ScalarParameter() {defaultValue=128, parameterName=Radius}, B=ScalarParameter() {defaultValue=64, parameterName=Height})
  )
  n89 = MaterialFunctionCall(Float2=@n88) {materialFunction=&BreakOutFloat2Components}
  n90 = Max(A=@n89.R, B=@n89.R)
  n91 = Max(A=@n88, B=Constant2Vector())
  Saturate(None=Add(A=Min(A=@n90) {constB=0}, B=Length(None=@n91)))
  SphereMask(A=ComponentMask() {r=true}, B=Constant() {r=0.5}, Hardness=Constant()) {attenuationRadius=0.025}
  n92 = NamedRerouteUsage()
  Subtract(A=OneMinus(None=ComponentMask(None=@n92) {g=true}), B=Multiply(A=@n39.R) {constB=0.5})
  Subtract(A=@n4, B=VectorParameter() {parameterName=CylinderPivotOffset, defaultValue=<rgba 0 0 64 1>}.RGB)
  TextureCoordinate()
  TextureCoordinate()
  TextureCoordinate()
  TextureCoordinate()
  TextureCoordinate()
  TextureCoordinate()
  TextureCoordinate()
  TextureCoordinate()
  TextureCoordinate()
  TextureCoordinate()
  TextureCoordinate()
  TextureSample() {samplerType=SAMPLERTYPE_Grayscale, texture=/Game/Effects/Fort_Effects/Textures/Fire/T_WispyFireNoise_Soft_M}
  n93 = Panner(Coordinate=Multiply(A=NamedRerouteUsage(), B=ScalarParameter() {defaultValue=4, parameterName=Tile})) {speedX=0.1, speedY=0.1}
  TextureSample(UVs=@n93) {
   samplerType=SAMPLERTYPE_LinearGrayscale
   texture=&T_Swirly_Noise_MiddleDetail
  }
  TextureSample(UVs=@n93) {texture=&T_SoftNoise}
  TextureSample() {texture=/Game/Global/Textures/Default/Tileable_Patterns/T_Thick_Smoke_05}
  TextureSample() {
   samplerType=SAMPLERTYPE_LinearGrayscale
   texture=&T_Swirly_Noise_MiddleDetail
  }
  TextureSample() {texture=&T_SoftNoise}
 ]
}
